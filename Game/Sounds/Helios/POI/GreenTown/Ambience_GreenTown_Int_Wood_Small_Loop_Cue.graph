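# Looping indoor ambience bed for small wooden interiors at the GreenTown POI.
# A quality-level node at the root selects the quad wave for the first two levels and the stereo version for the last two.
# Both waves are reused from the Apollo generic interior biome sounds and play through the indoor ambient sound class with shared ambient attenuation settings at reduced volume.
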SoundCue /Game/Sounds/Helios/POI/GreenTown/Ambience_GreenTown_Int_Wood_Small_Loop_Cue
root=QualityLevel_0
nodes [
 {id=QualityLevel_0, children [WavePlayer_0,WavePlayer_0,WavePlayer_1,WavePlayer_1]}
 {id=WavePlayer_0, properties {bLooping=true, soundWaveAssetPtr=/Game/Sounds/Apollo/Biomes/Interiors_Generic/AMB_Bed_Int_Wood_Small_01_Quad}}
 {id=WavePlayer_1, properties {bLooping=true, soundWaveAssetPtr=/Game/Sounds/Apollo/Biomes/Interiors_Generic/AMB_Bed_Int_Wood_Small_01_St}}
]
properties {
 attenuationSettings=&Fort_Ambient_Athena_Att
 maxDistance=2097152
 soundClassObject=&Fort_Ambient_Indoors_Athena
 volumeMultiplier=0.707
 attenuationOverrides {bAttenuate=false}
}
